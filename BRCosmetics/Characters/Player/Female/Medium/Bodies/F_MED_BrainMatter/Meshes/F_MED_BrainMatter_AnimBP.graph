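# Animation Blueprint for the BrainMatter female medium body cosmetic, adding deformation and secondary bow motion on top of the parent character's pose.
# Copy Pose From Mesh pulls the attached parent's pose into the F_MED_BrainMatter_DeformRig Control Rig.
# The result runs through six chained single-bone AnimDynamics nodes for the leg and back bow bones.
# These share identical angular springs, limits, gravity override and scaled component motion; the event graph has no logic.
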AnimBlueprint /BRCosmetics/Characters/Player/Female/Medium/Bodies/F_MED_BrainMatter/Meshes/F_MED_BrainMatter_AnimBP
target_skeleton=&/F_MED_BrainMatter_Skeleton
graphs [
 AnimGraph {nodes [
   n1 = Control Rig(Source=Copy Pose From Mesh() {bUseAttachedParent=true}) {
    alphaBoolBlend {blendOption=Linear}
    controlRigAssetReference {blueprintRigClass=&/F_MED_BrainMatter_DeformRig.F_MED_BrainMatter_DeformRig_C}
   }
   n2 = Anim Dynamics - Bone: dyn_bows_leg_l_01(ComponentPose=Local To Component(LocalPose=@n1), Alpha=1, ExternalForce="0,0,0") {
    angularDampingOverride=0.9995
    angularSpringConstant=2500
    bAngularSpring=true
    bGravityOverrideInSimSpace=true
    bOverrideAngularDamping=true
    bUseGravityOverride=true
    alphaBoolBlend {blendOption=Linear}
    boundBone {boneName=dyn_bows_leg_l_01}
    componentAppliedLinearAccClamp=<3000 3000 3000>
    componentLinearAccScale=<0.3 0.6 0.3>
    componentLinearVelScale=<0.3 0.3 0.3>
    gravityOverride {z=-980}
    physicsBodyDefinitions [{
      boundBone {boneName=dyn_bows_leg_l_01}
      boxExtents=<10 10 10>
      constraintSetup {angularTargetAxis=AxisZ, angularLimitsMax=<0 50 0>, angularLimitsMin=<0 -40 0>, angularTarget=<0 0 1>}
      localJointOffset=<0 0 0>
     }]
   }
   n3 = Anim Dynamics - Bone: dyn_bows_leg_l_02(ComponentPose=@n2, Alpha=1, ExternalForce="0,0,0") {
    angularDampingOverride=0.9995
    angularSpringConstant=2500
    bAngularSpring=true
    bGravityOverrideInSimSpace=true
    bOverrideAngularDamping=true
    bUseGravityOverride=true
    alphaBoolBlend {blendOption=Linear}
    boundBone {boneName=dyn_bows_leg_l_02}
    componentAppliedLinearAccClamp=<3000 3000 3000>
    componentLinearAccScale=<0.3 0.6 0.3>
    componentLinearVelScale=<0.3 0.3 0.3>
    gravityOverride {z=-980}
    physicsBodyDefinitions [{
      boundBone {boneName=dyn_bows_leg_l_02}
      boxExtents=<10 10 10>
      constraintSetup {angularTargetAxis=AxisZ, angularLimitsMax=<0 50 0>, angularLimitsMin=<0 -40 0>, angularTarget=<0 0 1>}
      localJointOffset=<0 0 0>
     }]
   }
   n4 = Anim Dynamics - Bone: dyn_bows_leg_r_01(ComponentPose=@n3, Alpha=1, ExternalForce="0,0,0") {
    angularDampingOverride=0.9995
    angularSpringConstant=2500
    bAngularSpring=true
    bGravityOverrideInSimSpace=true
    bOverrideAngularDamping=true
    bUseGravityOverride=true
    alphaBoolBlend {blendOption=Linear}
    boundBone {boneName=dyn_bows_leg_r_01}
    componentAppliedLinearAccClamp=<3000 3000 3000>
    componentLinearAccScale=<0.3 0.6 0.3>
    componentLinearVelScale=<0.3 0.3 0.3>
    gravityOverride {z=-980}
    physicsBodyDefinitions [{
      boundBone {boneName=dyn_bows_leg_r_01}
      boxExtents=<10 10 10>
      constraintSetup {angularTargetAxis=AxisZ, angularLimitsMax=<0 50 0>, angularLimitsMin=<0 -40 0>, angularTarget=<0 0 1>}
      localJointOffset=<0 0 0>
     }]
   }
   n5 = Anim Dynamics - Bone: dyn_bows_leg_r_02(ComponentPose=@n4, Alpha=1, ExternalForce="0,0,0") {
    angularDampingOverride=0.9995
    angularSpringConstant=2500
    bAngularSpring=true
    bGravityOverrideInSimSpace=true
    bOverrideAngularDamping=true
    bUseGravityOverride=true
    alphaBoolBlend {blendOption=Linear}
    boundBone {boneName=dyn_bows_leg_r_02}
    componentAppliedLinearAccClamp=<3000 3000 3000>
    componentLinearAccScale=<0.3 0.6 0.3>
    componentLinearVelScale=<0.3 0.3 0.3>
    gravityOverride {z=-980}
    physicsBodyDefinitions [{
      boundBone {boneName=dyn_bows_leg_r_02}
      boxExtents=<10 10 10>
      constraintSetup {angularTargetAxis=AxisZ, angularLimitsMax=<0 50 0>, angularLimitsMin=<0 -40 0>, angularTarget=<0 0 1>}
      localJointOffset=<0 0 0>
     }]
   }
   n6 = Anim Dynamics - Bone: dyn_backbeck_bow_01(ComponentPose=@n5, Alpha=1, ExternalForce="0,0,0") {
    angularDampingOverride=0.9995
    angularSpringConstant=2500
    bAngularSpring=true
    bGravityOverrideInSimSpace=true
    bOverrideAngularDamping=true
    bUseGravityOverride=true
    alphaBoolBlend {blendOption=Linear}
    boundBone {boneName=dyn_backbeck_bow_01}
    componentAppliedLinearAccClamp=<3000 3000 3000>
    componentLinearAccScale=<0.3 0.6 0.3>
    componentLinearVelScale=<0.3 0.3 0.3>
    gravityOverride {z=-980}
    physicsBodyDefinitions [{
      boundBone {boneName=dyn_backbeck_bow_01}
      boxExtents=<10 10 10>
      constraintSetup {angularTargetAxis=AxisZ, angularLimitsMax=<0 50 0>, angularLimitsMin=<0 -40 0>, angularTarget=<0 0 1>}
      localJointOffset=<0 0 0>
     }]
   }
   n7 = Anim Dynamics - Bone: dyn_backbeck_bow_02(ComponentPose=@n6, Alpha=1, ExternalForce="0,0,0") {
    angularDampingOverride=0.9995
    angularSpringConstant=2500
    bAngularSpring=true
    bGravityOverrideInSimSpace=true
    bOverrideAngularDamping=true
    bUseGravityOverride=true
    alphaBoolBlend {blendOption=Linear}
    boundBone {boneName=dyn_backbeck_bow_02}
    componentAppliedLinearAccClamp=<3000 3000 3000>
    componentLinearAccScale=<0.3 0.6 0.3>
    componentLinearVelScale=<0.3 0.3 0.3>
    gravityOverride {z=-980}
    physicsBodyDefinitions [{
      boundBone {boneName=dyn_backbeck_bow_02}
      boxExtents=<10 10 10>
      constraintSetup {angularTargetAxis=AxisZ, angularLimitsMax=<0 50 0>, angularLimitsMin=<0 -40 0>, angularTarget=<0 0 1>}
      localJointOffset=<0 0 0>
     }]
   }
   Output Pose(Result=Component To Local(ComponentPose=@n7)) {layerGroup=None}
  ]}
 EventGraph {nodes [Event BlueprintUpdateAnimation()]}
]
variables []
