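# Physics asset for the BrainMatter female character's backpack part, built around a dynamic dyn_hair_tail bone chain.
# Capsules on seven tail bones form a chain from spine_01, with tight twist and swing limits that mostly loosen along one swing axis toward the tip.
# Capsules on pelvis, spine_01, thighs and calves are linked by free angular constraints, apparently serving as body collision for the tail.
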PhysicsAsset /BRCosmetics/Characters/Player/Female/Medium/Bodies/F_MED_BrainMatter/Meshes/Parts/F_MED_BrainMatter_FaceAcc_BackPack_Physics
bodies {
 calf_l [{shapeType=Capsule, length=30, radius=9, shapeName=calf_l_capsule, center=<-20 0 0>, rotation=<rot 90 0 0>}]
 calf_r [{shapeType=Capsule, length=30, radius=9, shapeName=calf_r_capsule, center=<20 0 0>, rotation=<rot 90 0 0>}]
 dyn_hair_tail_01 [{shapeType=Capsule, length=8.56, radius=3.4197, shapeName=dyn_hair_tail_01_capsule, center=<10.3133 0 0>, rotation=<rot 0 -91.2536 -89.4576>}]
 dyn_hair_tail_02 [{shapeType=Capsule, length=7.0009, radius=5, shapeName=dyn_hair_tail_02_capsule, center=<5.7101 0 0>, rotation=<rot 0 90 -90>}]
 dyn_hair_tail_03 [{shapeType=Capsule, length=9.988, radius=5, shapeName=dyn_hair_tail_03_capsule, center=<10.9923 0 0>, rotation=<rot 0 90 -90>}]
 dyn_hair_tail_04 [{shapeType=Capsule, length=6.1869, radius=5, shapeName=dyn_hair_tail_04_capsule, center=<7.523 3.0605 0>, rotation=<rot 0 90 -90>}]
 dyn_hair_tail_05 [{shapeType=Capsule, length=8.2019, radius=5, shapeName=dyn_hair_tail_05_capsule, center=<6.9115 4 0>, rotation=<rot 0 90 -90>}]
 dyn_hair_tail_07 [{shapeType=Capsule, length=7.7213, radius=5, shapeName=dyn_hair_tail_07_capsule, center=<8.2322 3.5 0>, rotation=<rot 0 90 -90>}]
 dyn_hair_tail_08 [{shapeType=Capsule, length=6.6384, radius=5, shapeName=dyn_hair_tail_08_capsule, center=<6.0871 3.5 0>, rotation=<rot 0 90 -90>}]
 pelvis [{shapeType=Capsule, length=19.9809, radius=11.2908, shapeName=pelvis_capsule, center=<-6.0003 -4.3202 0>}]
 spine_01 [{shapeType=Capsule, length=17.7097, radius=9.911, shapeName=spine_01_capsule, center=<0 -2.1395 0>}]
 thigh_l [{shapeType=Capsule, length=30, radius=12, shapeName=thigh_l_capsule, center=<-20 0 0>, rotation=<rot 90 0 0>}]
 thigh_r [{shapeType=Capsule, length=30, radius=12, shapeName=thigh_r_capsule, center=<20 0 0>, rotation=<rot 90 0 0>}]
}
constraints [
 {bone1Name=dyn_hair_tail_02, bone2Name=dyn_hair_tail_01, swing1LimitDegrees=1, swing2LimitDegrees=30, twistLimitDegrees=5}
 {bone1Name=dyn_hair_tail_03, bone2Name=dyn_hair_tail_02, swing1LimitDegrees=5, swing2LimitDegrees=30, twistLimitDegrees=5}
 {bone1Name=dyn_hair_tail_04, bone2Name=dyn_hair_tail_03, swing1LimitDegrees=50, swing2LimitDegrees=20, twistLimitDegrees=5}
 {bone1Name=dyn_hair_tail_05, bone2Name=dyn_hair_tail_04, swing1LimitDegrees=50, swing2LimitDegrees=15, twistLimitDegrees=5}
 {bone1Name=dyn_hair_tail_07, bone2Name=dyn_hair_tail_05, swing1LimitDegrees=50, swing2LimitDegrees=15, twistLimitDegrees=5}
 {bone1Name=dyn_hair_tail_08, bone2Name=dyn_hair_tail_07, swing1LimitDegrees=50, swing2LimitDegrees=15, twistLimitDegrees=5}
 {bone1Name=dyn_hair_tail_01, bone2Name=spine_01, swing1LimitDegrees=0.5, swing2LimitDegrees=20, twistLimitDegrees=0}
 {
  bone1Name=spine_01
  bone2Name=pelvis
  swing1LimitDegrees=45
  swing1Motion=Free
  swing2LimitDegrees=45
  swing2Motion=Free
  twistLimitDegrees=45
  twistMotion=Free
 }
 {
  bone1Name=thigh_l
  bone2Name=pelvis
  swing1LimitDegrees=45
  swing1Motion=Free
  swing2LimitDegrees=45
  swing2Motion=Free
  twistLimitDegrees=45
  twistMotion=Free
 }
 {
  bone1Name=calf_l
  bone2Name=thigh_l
  swing1LimitDegrees=45
  swing1Motion=Free
  swing2LimitDegrees=45
  swing2Motion=Free
  twistLimitDegrees=45
  twistMotion=Free
 }
 {
  bone1Name=thigh_r
  bone2Name=pelvis
  swing1LimitDegrees=45
  swing1Motion=Free
  swing2LimitDegrees=45
  swing2Motion=Free
  twistLimitDegrees=45
  twistMotion=Free
 }
 {
  bone1Name=calf_r
  bone2Name=thigh_r
  swing1LimitDegrees=45
  swing1Motion=Free
  swing2LimitDegrees=45
  swing2Motion=Free
  twistLimitDegrees=45
  twistMotion=Free
 }
]
