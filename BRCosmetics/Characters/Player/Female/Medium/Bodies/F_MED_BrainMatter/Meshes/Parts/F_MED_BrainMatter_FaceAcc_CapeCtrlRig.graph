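# Control Rig for the female BrainMatter face accessory (named as a cape rig) that corrects the hair tail's base bone as the head tilts.
# It takes the head's rotation relative to spine_05 as Euler Z and remaps it into clamped weights for negative and positive tilt.
# Those weights drive two chained Modify Transforms applying opposite Z rotations to dyn_hair_tail_base.
# A Spherical Pose Reader on the head is present but unconnected.
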ControlRigBlueprint /BRCosmetics/Characters/Player/Female/Medium/Bodies/F_MED_BrainMatter/Meshes/Parts/F_MED_BrainMatter_FaceAcc_CapeCtrlRig
graphs [
 RigVMFunctionLibrary {links [], nodes []}
 RigVMModel {nodes [
   RigUnit_BeginExecution = &RigUnit_BeginExecution()
   RigUnit_GetRelativeTransformForItem = &RigUnit_GetRelativeTransformForItem(Child=(Type=Bone,Name="head"), Parent=(Type=Bone,Name="spine_05"))
   RigVMFunction_MathQuaternionToEuler = &RigVMFunction_MathQuaternionToEuler(RotationOrder=ZYX, Value=@RigUnit_GetRelativeTransformForItem.RelativeTransform.Rotation)
   Remap = &RigVMFunction_MathFloatRemap(SourceMaximum=-38, SourceMinimum=-5, TargetMinimum=0, bClamp=true, Value=@RigVMFunction_MathQuaternionToEuler.Result.Z)
   RigVMFunction_MathQuaternionFromEuler = &RigVMFunction_MathQuaternionFromEuler(Euler=(X=0,Y=0,Z=40))
   &RigUnit_ModifyTransforms(
    ItemToModify=((Item=(Type=Bone,Name="dyn_hair_tail_base"),Transform=(Rotation=(X=0,Y=0,Z=0,W=1),Translation=(X=0,Y=0,Z=0),Scale3D=(X=1,Y=1,Z=1))))
    Weight=@Remap.Result
    ExecutePin=@RigUnit_BeginExecution.ExecutePin
    ItemToModify.0.Transform.Rotation=@RigVMFunction_MathQuaternionFromEuler.Result
   )
   > &RigUnit_ModifyTransforms(
    ItemToModify=((Item=(Type=Bone,Name="dyn_hair_tail_base"),Transform=(Rotation=(X=0,Y=0,Z=0,W=1),Translation=(X=0,Y=0,Z=0),Scale3D=(X=1,Y=1,Z=1))))
    Weight=&RigVMFunction_MathFloatRemap(SourceMaximum=20, SourceMinimum=10, TargetMinimum=0, bClamp=true, Value=@RigVMFunction_MathQuaternionToEuler.Result.Z).Result
    ItemToModify.0.Transform.Rotation=&RigVMFunction_MathQuaternionFromEuler(Euler=(X=0,Y=0,Z=-10)).Result
   )
   &RigUnit_SphericalPoseReader(
    ActiveRegionScaleFactors=(PositiveWidth=1,NegativeWidth=1,PositiveHeight=1,NegativeHeight=1)
    ActiveRegionSize=0.1
    Debug=(bDrawDebug=True,DebugScale=25,DebugSegments=20,DebugThickness=0.25)
    DriverAxis=(X=1,Y=0,Z=0)
    DriverItem=(Type=Bone,Name="head")
    ExecutePin=()
    FalloffRegionScaleFactors=(PositiveWidth=1,NegativeWidth=1,PositiveHeight=1,NegativeHeight=1)
    FalloffSize=0.2
    OptionalParentItem=(Type=Bone,Name="spine_05")
    RotationOffset=(X=-90,Y=0,Z=30)
   )
  ]}
]
hierarchy [
 root {loc=<0 0 0>, rot=<0 0 0 1>}
 pelvis {parent=root, loc=<0 -0.5301 98.0571>, rot=<0.0127 0.707 -0.0127 -0.707>}
 spine_01 {parent=pelvis, loc=<6.1208 -0.9328 0>, rot=<0 0 0.034 -0.9994>}
 spine_02 {parent=spine_01, loc=<6.5996 -0.2095 0>, rot=<0 0 0.0238 -0.9997>}
 spine_03 {parent=spine_02, loc=<6.4914 -0.1237 0>, rot=<0 0 0.0149 -0.9999>}
 spine_04 {parent=spine_03, loc=<6.9598 0.0794 0>, rot=<0 0 -0.0132 -0.9999>}
 spine_05 {parent=spine_04, loc=<6.9987 0 0>, rot=<0 0 -0.0015 -1>}
 neck_01 {parent=spine_05, loc=<11.3138 0.1529 0>, rot=<0 0 -0.2003 -0.9797>}
 neck_02 {parent=neck_01, loc=<4.5627 0 0>, rot=<0 0 0.0504 -0.9987>}
 head {parent=neck_02, loc=<5.3914 0 0>, rot=<0 0 0.115 -0.9934>}
 dyn_earring_l {parent=head, loc=<0 -0.5 -6.3>, rot=<-0.0111 -0.9847 0.1737 -0.0007>}
 dyn_earring_r {parent=head, loc=<0 -0.5 6.3>, rot=<-0.0045 -0.1743 0.9846 0.0102>}
 dyn_hair_locan_r_01 {parent=head, loc=<4.0656 2.3199 6.6784>, rot=<0.1678 -0.7192 0.6742 -0.0081>}
 dyn_hair_locan_r_02 {parent=dyn_hair_locan_r_01, loc=<3.1764 -0.0175 0>, rot=<0 0 -0.0305 -0.9995>}
 dyn_hair_locan_r_04 {parent=dyn_hair_locan_r_02, loc=<3.837 0 0>, rot=<0 0 0 1>}
 dyn_hair_locan_l_01 {parent=head, loc=<4.066 2.32 -6.678>, rot=<0.1678 -0.7192 -0.6742 0.0081>}
 dyn_hair_locan_l_02 {parent=dyn_hair_locan_l_01, loc=<3.321 0 0>, rot=<0 0 -0.0305 -0.9995>}
 dyn_hair_locan_l_04 {parent=dyn_hair_locan_l_02, loc=<3.253 -0.1664 0.0653>, rot=<0 0 0 1>}
 Scale_Hair {parent=head, loc=<9.3235 -2.0656 0>, rot=<0 0 0.2658 -0.964>}
 dyn_hair_tail_base {parent=head, loc=<13.8166 -13.5667 0>, rot=<-0.0041 -0.0005 0.992 -0.1258>}
 dyn_hair_tail_01 {parent=dyn_hair_tail_base, loc=<0 0 0>, rot=<0 0 0 1>}
 dyn_hair_tail_02 {parent=dyn_hair_tail_01, loc=<19.863 0 0>, rot=<0 0 0.0608 -0.9982>}
 dyn_hair_tail_03 {parent=dyn_hair_tail_02, loc=<13.1926 -1.618 -0.0347>, rot=<0 0 0.055 -0.9985>}
 dyn_hair_tail_04 {parent=dyn_hair_tail_03, loc=<20.5564 -1.9899 -0.0427>, rot=<0.0063 -0.0008 -0.1229 -0.9924>}
 dyn_hair_tail_05 {parent=dyn_hair_tail_04, loc=<14.3547 0 0>, rot=<-0.004 0.0083 0.0576 -0.9983>}
 dyn_hair_tail_07 {parent=dyn_hair_tail_05, loc=<15.4773 0 0>, rot=<0 0 0.1182 -0.993>}
 dyn_hair_tail_08 {parent=dyn_hair_tail_07, loc=<13.9643 0 0>, rot=<0 0 0.0436 -0.999>}
 clavicle_l {parent=spine_05, loc=<5.4831 -0.1053 -3.2946>, rot=<0.0465 -0.7289 -0.009 -0.6829>}
 upperarm_l {parent=clavicle_l, loc=<9.8658 -1.2136 -0.4105>, rot=<0.0164 -0.3245 -0.0269 -0.9454>}
 lowerarm_l {parent=upperarm_l, loc=<26.465 -0.8345 0.259>, rot=<0.0036 0.0161 -0.2298 -0.9731>}
 clavicle_r {parent=spine_05, loc=<5.4831 -0.1053 3.2946>, rot=<0.7289 0.0465 -0.6829 0.009>}
 upperarm_r {parent=clavicle_r, loc=<-9.8658 1.2136 0.4105>, rot=<0.0164 -0.3245 -0.0269 -0.9454>}
 lowerarm_r {parent=upperarm_r, loc=<-26.465 0.8345 -0.259>, rot=<0.0036 0.0161 -0.2298 -0.9731>}
 Hair_AltSkin_Corrective_01 {parent=spine_05, loc=<3.3389 -16.3891 -0.3065>, rot=<0.0284 -0.7065 -0.0284 -0.7065>}
 Hair_AltSkin_Corrective_02 {parent=spine_02, loc=<3.1294 -16.5787 -0.5114>, rot=<0.0282 -0.7065 -0.0282 -0.7065>}
 thigh_l {parent=pelvis, loc=<-1.0855 -0.2109 -8.7053>, rot=<0.0009 0.0355 0.0155 -0.9993>}
 calf_l {parent=thigh_l, loc=<-42.5284 -0.0019 -0.7297>, rot=<0.0006 -0.0161 -0.0391 -0.9991>}
 thigh_r {parent=pelvis, loc=<-1.0855 -0.2109 8.7053>, rot=<0.0355 -0.0009 0.9993 0.0155>}
 calf_r {parent=thigh_r, loc=<42.5284 0.0019 0.7297>, rot=<0.0006 -0.0161 -0.0391 -0.9991>}
 Hair_AltSkin_Corrective_03 {parent=pelvis, loc=<-0.1109 -21.5957 -0.675>, rot=<-0.0127 -0.707 0.0127 -0.707>}
]
