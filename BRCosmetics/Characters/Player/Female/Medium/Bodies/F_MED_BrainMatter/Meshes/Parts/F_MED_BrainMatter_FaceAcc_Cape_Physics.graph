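# Physics asset for the BrainMatter female medium body's FaceAcc cape mesh, covering a dyn_hair_tail bone chain and the body around it.
# Seven tail capsules link in series with narrow swing limits and zero twist, skipping tail_06, from a small base capsule locked to the head.
# Capsules on head, spine, pelvis, arms and thighs form body volumes with free angular constraints, plus an oversized root capsule offset downward.
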PhysicsAsset /BRCosmetics/Characters/Player/Female/Medium/Bodies/F_MED_BrainMatter/Meshes/Parts/F_MED_BrainMatter_FaceAcc_Cape_Physics
bodies {
 dyn_hair_tail_01 [{shapeType=Capsule, length=8.56, radius=3.4197, shapeName=dyn_hair_tail_01_capsule, center=<10.3133 0 0>, rotation=<rot 0 90 -90>}]
 dyn_hair_tail_02 [{shapeType=Capsule, length=9.0082, radius=2.9927, shapeName=dyn_hair_tail_02_capsule, center=<5.7101 0 0>, rotation=<rot 0 90 -90>}]
 dyn_hair_tail_03 [{
   shapeType=Capsule
   length=12.4298
   radius=2.5581
   shapeName=dyn_hair_tail_03_capsule
   center=<10.8746 -0.8521 -0.0091>
   rotation=<rot 0.6038 90 -90>
  }]
 dyn_hair_tail_04 [{shapeType=Capsule, length=8.4667, radius=2.7202, shapeName=dyn_hair_tail_04_capsule, center=<7.5272 0 0>, rotation=<rot 0 90 -90>}]
 dyn_hair_tail_05 [{shapeType=Capsule, length=11.0857, radius=2.1163, shapeName=dyn_hair_tail_05_capsule, center=<6.8892 0 0>, rotation=<rot 0 90 -90>}]
 dyn_hair_tail_07 [{shapeType=Capsule, length=10.5972, radius=2.1241, shapeName=dyn_hair_tail_07_capsule, center=<8.1359 0 0>, rotation=<rot 0 90 -90>}]
 dyn_hair_tail_08 [{shapeType=Capsule, length=9.4601, radius=2.1783, shapeName=dyn_hair_tail_08_capsule, center=<6.1592 0 0>, rotation=<rot 0 90 -90>}]
 dyn_hair_tail_base [{shapeType=Capsule, length=0.505, radius=0.505, shapeName=dyn_hair_tail_base_capsule, rotation=<rot 0 -90 0>}]
 head [{
   shapeType=Capsule
   length=11.3986
   radius=8.9621
   shapeName=head_capsule
   center=<11.1858 -4.8745 -0.337>
   rotation=<rot 0 -131.1778 -88.2001>
  }]
 lowerarm_l [{shapeType=Capsule, length=20, radius=6, shapeName=lowerarm_l_capsule, center=<15 0 0>, rotation=<rot 90 0 0>}]
 lowerarm_r [{shapeType=Capsule, length=20, radius=6, shapeName=lowerarm_r_capsule, center=<-15 0 0>, rotation=<rot 90 0 0>}]
 pelvis [{shapeType=Capsule, length=11.5933, radius=19.6783, shapeName=pelvis_capsule, center=<-4.92 -7.0625 0>}]
 root [{shapeType=Capsule, length=99.3415, radius=116.1049, shapeName=root_capsule, center=<0 0 -112.3973>, rotation=<rot 90 90 180>}]
 spine_01 [{shapeType=Capsule, length=10.5342, radius=17.0865, shapeName=spine_01_capsule, center=<3.521 -5.6472 0>}]
 spine_04 [{shapeType=Capsule, length=18.9818, radius=15.2052, shapeName=spine_04_capsule, center=<0 -4.2719 0>}]
 spine_05 [{shapeType=Capsule, length=19.7232, radius=7.8975, shapeName=spine_05_capsule, center=<6.4158 -3.5179 0>}]
 thigh_l [{shapeType=Capsule, length=30, radius=12, shapeName=thigh_l_capsule, center=<-20 0 0>, rotation=<rot 90 0 0>}]
 thigh_r [{shapeType=Capsule, length=30, radius=12, shapeName=thigh_r_capsule, center=<20 0 0>, rotation=<rot 90 0 0>}]
 upperarm_l [{shapeType=Capsule, length=20, radius=7, shapeName=upperarm_l_capsule, center=<15 0 0>, rotation=<rot 90 0 0>}]
 upperarm_r [{shapeType=Capsule, length=20, radius=7, shapeName=upperarm_r_capsule, center=<-15 0 0>, rotation=<rot 90 0 0>}]
}
constraints [
 {bone1Name=dyn_hair_tail_02, bone2Name=dyn_hair_tail_01, swing1LimitDegrees=6.36, swing2LimitDegrees=15, twistLimitDegrees=0}
 {bone1Name=dyn_hair_tail_03, bone2Name=dyn_hair_tail_02, swing1LimitDegrees=9.24, swing2LimitDegrees=15, twistLimitDegrees=0}
 {bone1Name=dyn_hair_tail_04, bone2Name=dyn_hair_tail_03, swing1LimitDegrees=7.8, swing2LimitDegrees=15, twistLimitDegrees=0}
 {bone1Name=dyn_hair_tail_05, bone2Name=dyn_hair_tail_04, swing1LimitDegrees=7.8, swing2LimitDegrees=15, twistLimitDegrees=0}
 {bone1Name=dyn_hair_tail_07, bone2Name=dyn_hair_tail_05, swing1LimitDegrees=14.92, swing2LimitDegrees=25, twistLimitDegrees=0}
 {bone1Name=dyn_hair_tail_08, bone2Name=dyn_hair_tail_07, swing1LimitDegrees=13.48, swing2LimitDegrees=25, twistLimitDegrees=0}
 {bone1Name=dyn_hair_tail_01, bone2Name=dyn_hair_tail_base, swing1LimitDegrees=15.68, swing2LimitDegrees=25, twistLimitDegrees=0}
 {
  bone1Name=spine_01
  bone2Name=pelvis
  swing1LimitDegrees=45
  swing1Motion=Free
  swing2LimitDegrees=45
  swing2Motion=Free
  twistLimitDegrees=45
  twistMotion=Free
 }
 {
  bone1Name=spine_04
  bone2Name=spine_01
  swing1LimitDegrees=45
  swing1Motion=Free
  swing2LimitDegrees=45
  swing2Motion=Free
  twistLimitDegrees=45
  twistMotion=Free
 }
 {
  bone1Name=spine_05
  bone2Name=spine_04
  swing1LimitDegrees=45
  swing1Motion=Free
  swing2LimitDegrees=45
  swing2Motion=Free
  twistLimitDegrees=45
  twistMotion=Free
 }
 {
  bone1Name=upperarm_l
  bone2Name=spine_05
  swing1LimitDegrees=45
  swing1Motion=Free
  swing2LimitDegrees=45
  swing2Motion=Free
  twistLimitDegrees=45
  twistMotion=Free
 }
 {
  bone1Name=lowerarm_l
  bone2Name=upperarm_l
  swing1LimitDegrees=45
  swing1Motion=Free
  swing2LimitDegrees=45
  swing2Motion=Free
  twistLimitDegrees=45
  twistMotion=Free
 }
 {
  bone1Name=upperarm_r
  bone2Name=spine_05
  swing1LimitDegrees=45
  swing1Motion=Free
  swing2LimitDegrees=45
  swing2Motion=Free
  twistLimitDegrees=45
  twistMotion=Free
 }
 {
  bone1Name=lowerarm_r
  bone2Name=upperarm_r
  swing1LimitDegrees=45
  swing1Motion=Free
  swing2LimitDegrees=45
  swing2Motion=Free
  twistLimitDegrees=45
  twistMotion=Free
 }
 {
  bone1Name=thigh_l
  bone2Name=pelvis
  swing1LimitDegrees=45
  swing1Motion=Free
  swing2LimitDegrees=45
  swing2Motion=Free
  twistLimitDegrees=45
  twistMotion=Free
 }
 {
  bone1Name=thigh_r
  bone2Name=pelvis
  swing1LimitDegrees=45
  swing1Motion=Free
  swing2LimitDegrees=45
  swing2Motion=Free
  twistLimitDegrees=45
  twistMotion=Free
 }
 {
  bone1Name=pelvis
  bone2Name=root
  swing1LimitDegrees=45
  swing1Motion=Free
  swing2LimitDegrees=45
  swing2Motion=Free
  twistLimitDegrees=45
  twistMotion=Free
 }
 {
  bone1Name=head
  bone2Name=spine_05
  swing1LimitDegrees=45
  swing1Motion=Free
  swing2LimitDegrees=45
  swing2Motion=Free
  twistLimitDegrees=45
  twistMotion=Free
 }
 {
  bone1Name=dyn_hair_tail_base
  bone2Name=head
  swing1LimitDegrees=45
  swing1Motion=Locked
  swing2LimitDegrees=45
  swing2Motion=Locked
  twistLimitDegrees=45
  twistMotion=Locked
 }
]
