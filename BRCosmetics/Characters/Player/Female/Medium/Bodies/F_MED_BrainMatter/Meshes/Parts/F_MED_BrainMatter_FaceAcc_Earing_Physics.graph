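# Physics asset for the earring face accessory of the female medium BrainMatter player character, covering dangling earrings and side hair locks.
# Each earring body pairs two capsules and hangs from the head with limited swing.
# Left and right hair chains link three capsule bodies (01, 02, 04, skipping 03) from the head with tighter swing limits.
# Head and neck_02 capsules provide the base bodies, joined by a constraint with free swing and twist.
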PhysicsAsset /BRCosmetics/Characters/Player/Female/Medium/Bodies/F_MED_BrainMatter/Meshes/Parts/F_MED_BrainMatter_FaceAcc_Earing_Physics
bodies {
 dyn_earring_l [
  {
   shapeType=Capsule
   length=9.3499
   radius=0.5
   shapeName=dyn_earring_l_capsule
   center=<2.3371 0.6823 -0.0183>
   rotation=<rot -89.801 179.9999 -73.8929>
  }
  {
   shapeType=Capsule
   length=9.3499
   radius=0.5
   shapeName=dyn_earring_l_capsule
   center=<3.1294 -0.1451 -0.0211>
   rotation=<rot -89.801 179.9997 174.0519>
  }
 ]
 dyn_earring_r [
  {
   shapeType=Capsule
   length=9.3499
   radius=0.5
   shapeName=dyn_earring_l_capsule
   center=<4.1321 0.0595 -0.0246>
   rotation=<rot -89.7947 179.9558 -178.8641>
  }
  {
   shapeType=Capsule
   length=9.3499
   radius=0.5
   shapeName=dyn_earring_l_capsule
   center=<2.3133 0.2257 -0.0183>
   rotation=<rot -89.801 179.9996 -101.6618>
  }
 ]
 dyn_hair_locan_l_01 [{shapeType=Capsule, length=2.2, radius=0.5, shapeName=dyn_hair_locan_l_01_capsule, rotation=<rot 90 0 0>}]
 dyn_hair_locan_l_02 [{shapeType=Capsule, length=4.2177, radius=0.5, shapeName=dyn_hair_locan_l_02_capsule, center=<0 -0.4 0>, rotation=<rot 90 0 0>}]
 dyn_hair_locan_l_04 [{shapeType=Capsule, length=4.3894, radius=0.5, shapeName=dyn_hair_locan_l_04_capsule, center=<0 -0.4 0>, rotation=<rot 90 0 0>}]
 dyn_hair_locan_r_01 [{shapeType=Capsule, length=2.2, radius=0.5, shapeName=dyn_hair_locan_r_01_capsule, rotation=<rot 90 0 0>}]
 dyn_hair_locan_r_02 [{shapeType=Capsule, length=4.2966, radius=0.5, shapeName=dyn_hair_locan_r_02_capsule, center=<0 -0.4 0>, rotation=<rot 90 0 0>}]
 dyn_hair_locan_r_04 [{shapeType=Capsule, length=4.3786, radius=0.5, shapeName=dyn_hair_locan_r_04_capsule, center=<0 -0.4 0>, rotation=<rot 90 0 0>}]
 head [{
   shapeType=Capsule
   length=4.5635
   radius=6.4754
   shapeName=head_capsule
   center=<2.9414 2.117 -0.0152>
   rotation=<rot -0.4573 -87.6705 -89.4105>
  }]
 neck_02 [{shapeType=Capsule, length=4.3484, radius=4.8997, shapeName=neck_02_capsule, rotation=<rot 90 0 0>}]
}
constraints [
 {bone1Name=dyn_earring_l, bone2Name=head, swing1LimitDegrees=25, swing2LimitDegrees=25, twistLimitDegrees=0}
 {bone1Name=dyn_earring_r, bone2Name=head, swing1LimitDegrees=25, swing2LimitDegrees=25, twistLimitDegrees=0}
 {bone1Name=dyn_hair_locan_l_01, bone2Name=head, swing1LimitDegrees=10, swing2LimitDegrees=10, twistLimitDegrees=0}
 {bone1Name=dyn_hair_locan_l_02, bone2Name=dyn_hair_locan_l_01, swing1LimitDegrees=10, swing2LimitDegrees=10, twistLimitDegrees=0}
 {bone1Name=dyn_hair_locan_l_04, bone2Name=dyn_hair_locan_l_02, swing1LimitDegrees=10, swing2LimitDegrees=10, twistLimitDegrees=0}
 {bone1Name=dyn_hair_locan_r_01, bone2Name=head, swing1LimitDegrees=10, swing2LimitDegrees=10, twistLimitDegrees=0}
 {bone1Name=dyn_hair_locan_r_02, bone2Name=dyn_hair_locan_r_01, swing1LimitDegrees=10, swing2LimitDegrees=10, twistLimitDegrees=0}
 {bone1Name=dyn_hair_locan_r_04, bone2Name=dyn_hair_locan_r_02, swing1LimitDegrees=10, swing2LimitDegrees=10, twistLimitDegrees=0}
 {
  bone1Name=head
  bone2Name=neck_02
  swing1LimitDegrees=45
  swing1Motion=Free
  swing2LimitDegrees=45
  swing2Motion=Free
  twistLimitDegrees=45
  twistMotion=Free
 }
]
