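# Post-process-style Anim Blueprint for the BrainMatter female face accessory mesh, animating hair tail, earring, backpack and cape secondary motion.
# Copied parent and body poses pass through Control Rigs and bone offsets for backbling, jacket variant and skydive, then four chained RigidBody simulations.
# A front-end curve blends slot animation onto the hair.
# The EventGraph detects cosmetics and movement states, choosing gravity/wind parameters for cape wear.
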AnimBlueprint /BRCosmetics/Characters/Player/Female/Medium/Bodies/F_MED_BrainMatter/Meshes/Parts/F_MED_BrainMatter_FaceAcc_AnimBP
target_skeleton=&/F_MED_BrainMatter_FaceAcc_Skeleton
graphs [
 AnimGraph {nodes [
   n1 = "Transform (Modify) Bone - Bone: dyn_hair_tail_base"!(
    ComponentPose=Local To Component(LocalPose=Use cached pose 'Pre sim'())
    bAlphaBoolEnabled:bool=True
    Translation="0,0,0"
    Rotation="0,0,60"
    Scale="1,1,1"
   ) -> (Pose) {
    alphaInputType=Bool
    rotationMode=BMM_Additive
    alphaBoolBlend {blendInTime=0.5, blendOption=ExpIn, blendOutTime=0.5}
    alphaScaleBiasClamp {bInterpResult=true, interpSpeedDecreasing=6}
    boneToModify {boneName=dyn_hair_tail_base}
    rotation {roll=60}
   }
   n2 = "Transform (Modify) Bone - Bone: dyn_hair_tail_base"!(ComponentPose=@n1, Alpha:float=1, Translation="0,0,0", Rotation="0,10,0", Scale="1,1,1") -> (Pose) {
    rotationMode=BMM_Additive
    rotationSpace=BCS_BoneSpace
    translationMode=BMM_Additive
    translationSpace=BCS_BoneSpace
    alphaBoolBlend {blendOption=Linear}
    boneToModify {boneName=dyn_hair_tail_base}
    rotation {yaw=10}
   }
   n3 = Component To Local(ComponentPose=@n2)
   n4 = "Transform (Modify) Bone - Bone: dyn_hair_tail_base"(
    ComponentPose=Local To Component(LocalPose=Use cached pose 'Pre sim'())
    Alpha=1
    Translation="0,0,0"
    Rotation="-0.081,0.842,0.010"
    Scale="1,1,1"
   ) {
    rotationMode=BMM_Additive
    rotationSpace=BCS_BoneSpace
    translationMode=BMM_Additive
    translationSpace=BCS_BoneSpace
    alphaBoolBlend {blendOption=Linear}
    boneToModify {boneName=dyn_hair_tail_base}
    rotation=<rot -0.081 0.842 0.01>
   }
   n5 = Component To Local(ComponentPose=@n4)
   n6 = RigidBody!(
    ComponentPose=Local To Component(LocalPose=Blend Poses by bool(BlendPose_0=@n3, BlendPose_1=@n5) {blendType=Linear})
    Alpha
    OverrideWorldGravity:Vector="0,0,-920"
    ExternalForce
   ) -> (Pose) {
    bClampLinearTranslationLimitToRefPose=true
    bOverrideWorldGravity=true
    overridePhysicsAsset=&/F_MED_BrainMatter_FaceAcc_Physics
    simulationSpace=BaseBoneSpace
    alphaBoolBlend {blendOption=Linear}
    baseBoneRef {boneName=spine_02}
    componentAppliedLinearAccClamp=<3000 3000 3000>
    componentLinearAccScale=<0.3 0.25 0.5>
    componentLinearVelScale=<0.3 0.25 0.5>
    overrideWorldGravity {z=-920}
    simSpaceSettings {maxAngularAcceleration=40, maxAngularVelocity=3, maxLinearAcceleration=40, maxLinearVelocity=50, velocityScaleZ=0, worldAlpha=1}
   }
   n7 = RigidBody(ComponentPose=@n6, OverrideWorldGravity="0,0,-920") {
    bClampLinearTranslationLimitToRefPose=true
    bOverrideWorldGravity=true
    overridePhysicsAsset=&/F_MED_BrainMatter_FaceAcc_BackPack_Physics
    simulationSpace=BaseBoneSpace
    alphaBoolBlend {blendOption=Linear}
    baseBoneRef {boneName=spine_02}
    overrideWorldGravity {z=-920}
    simSpaceSettings {
     maxAngularAcceleration=1000
     maxAngularVelocity=3
     maxLinearAcceleration=500
     maxLinearVelocity=250
     velocityScaleZ=0
     worldAlpha=1
     externalLinearDragV=<0.4 0.4 0.4>
    }
   }
   n8 = RigidBody!(ComponentPose=@n7, bAlphaBoolEnabled:bool, OverrideWorldGravity="0,0,-920", ExternalForce) -> (Pose) {
    alphaInputType=Bool
    bClampLinearTranslationLimitToRefPose=true
    bOverrideWorldGravity=true
    overridePhysicsAsset=&/F_MED_BrainMatter_FaceAcc_Cape_Physics
    simulationSpace=BaseBoneSpace
    alphaBoolBlend {blendOption=Linear}
    baseBoneRef {boneName=spine_02}
    overrideWorldGravity {z=-920}
    simSpaceSettings {
     maxAngularAcceleration=1000
     maxAngularVelocity=3
     maxLinearAcceleration=500
     maxLinearVelocity=250
     velocityScaleZ=0
     worldAlpha=1
     externalLinearDragV=<0.4 0.4 0.4>
    }
   }
   n9 = RigidBody!(ComponentPose=@n8, Alpha:float, OverrideWorldGravity="0,0,0", ExternalForce="0,0,3000") -> (Pose) {
    bClampLinearTranslationLimitToRefPose=true
    bOverrideWorldGravity=true
    overridePhysicsAsset=&/F_MED_BrainMatter_FaceAcc_Earing_Physics
    simulationSpace=BaseBoneSpace
    alphaBoolBlend {blendOption=Linear}
    baseBoneRef {boneName=head}
    componentLinearAccScale=<0.2 0.2 0.2>
    simSpaceSettings {
     maxAngularAcceleration=1000
     maxAngularVelocity=3
     maxLinearAcceleration=500
     maxLinearVelocity=250
     velocityScaleZ=0
     worldAlpha=1
     externalLinearDragV=<0.4 0.4 0.4>
    }
   }
   Output Pose(Result=Layered blend per bone(BasePose=Component To Local(ComponentPose=@n9), BlendPoses_0=Slot 'DefaultSlot'(), BlendWeights_0=Get FEIdleFloat() -> (FEIdleFloat:double)) {layerSetup [{branchFilters [{boneName=dyn_hair_tail_base}]}]}) {layerGroup=None}
   n10 = Layered blend per bone(BasePose=Copy Pose From Mesh() {bUseAttachedParent=true}, BlendPoses_0=Copy Pose From Mesh(SourceMeshComponent=Get Body_Mesh())) {layerSetup [{branchFilters [
       {boneName=upperarm_l}
       {boneName=upperarm_r}
       {boneName=lowerarm_l}
       {boneName=lowerarm_r}
       {boneName=thigh_l}
       {boneName=thigh_r}
       {boneName=calf_l}
       {boneName=calf_r}
      ]}]}
   n11 = Control Rig!(Source=@n10, bAlphaBoolEnabled:bool=Get IsWearingBackbling() -> (IsWearingBackbling:bool)) -> (Pose) {
    alphaInputType=Bool
    alphaBoolBlend {blendOption=Linear}
    controlRigAssetReference {blueprintRigClass=&/F_MED_BrainMatter_FaceAcc_CtrlRig.F_MED_BrainMatter_FaceAcc_CtrlRig_C}
   }
   n12 = Control Rig(Source=@n11, bAlphaBoolEnabled=Get is Wear a Cape() -> (is Wear a Cape:bool)) {
    alphaInputType=Bool
    alphaBoolBlend {blendOption=Linear}
    controlRigAssetReference {blueprintRigClass=&/F_MED_BrainMatter_FaceAcc_CapeCtrlRig.F_MED_BrainMatter_FaceAcc_CapeCtrlRig_C}
   }
   n13 = Control Rig!(Source=@n12, bAlphaBoolEnabled=NOT Boolean(A=Get bRigidBodyEnabled()), Tactical sprint:bool=Get bIsTacticalSprinting()) -> (Pose) {
    alphaInputType=Bool
    alphaBoolBlend {blendOption=Linear}
    controlRigAssetReference {blueprintRigClass=&/F_MED_BrainMatter_FaceAcc_AltSkinCtrlRig.F_MED_BrainMatter_FaceAcc_AltSkinCtrlRig_C}
   }
   n14 = "Transform (Modify) Bone - Bone: dyn_hair_tail_base"!(ComponentPose=Local To Component(LocalPose=@n13), bAlphaBoolEnabled:bool=True, Translation="-4,5,0", Rotation="0,0,0", Scale="1,1,1") -> (Pose) {
    alphaInputType=Bool
    translationMode=BMM_Additive
    translationSpace=BCS_BoneSpace
    alphaBoolBlend {blendInTime=0.5, blendOption=Linear, blendOutTime=0.5}
    boneToModify {boneName=dyn_hair_tail_base}
    translation=<-4 5>
   }
   n15 = "Transform (Modify) Bone - Bone: dyn_hair_tail_base"(ComponentPose=@n14, bAlphaBoolEnabled=True, Translation="0,0,0", Rotation="0,30,0", Scale="1,1,1") {
    alphaInputType=Bool
    rotationMode=BMM_Additive
    rotationSpace=BCS_BoneSpace
    alphaBoolBlend {blendInTime=0.5, blendOption=Linear, blendOutTime=0.5}
    boneToModify {boneName=dyn_hair_tail_base}
    rotation {yaw=30}
   }
   n16 = "Transform (Modify) Bone - Bone: dyn_hair_tail_base"(ComponentPose=@n15, bAlphaBoolEnabled=True, Translation="0,0,0", Rotation="0,30,0", Scale="1,1,1") {
    alphaInputType=Bool
    rotationMode=BMM_Additive
    rotationSpace=BCS_BoneSpace
    alphaBoolBlend {blendInTime=0.5, blendOption=Linear, blendOutTime=0.7}
    boneToModify {boneName=dyn_hair_tail_base}
    rotation {yaw=30}
   }
   n17 = "Transform (Modify) Bone - Bone: dyn_hair_tail_base"(ComponentPose=@n16, bAlphaBoolEnabled=True, Translation="0,0,0", Rotation="0,20,0", Scale="1,1,1") {
    alphaInputType=Bool
    rotationMode=BMM_Additive
    rotationSpace=BCS_BoneSpace
    alphaBoolBlend {blendInTime=0.5, blendOption=Linear, blendOutTime=0.5}
    boneToModify {boneName=dyn_hair_tail_base}
    rotation {yaw=20}
   }
   n18 = "Transform (Modify) Bone - Bone: Scale_Hair"!(ComponentPose=@n17, bAlphaBoolEnabled:bool=True, Translation="0.452,0,0", Rotation="0,0,0", Scale="1.5,1.5,1.5") -> (Pose) {
    alphaInputType=Bool
    scaleMode=BMM_Additive
    scaleSpace=BCS_BoneSpace
    translationMode=BMM_Additive
    translationSpace=BCS_BoneSpace
    alphaBoolBlend {blendOption=Linear}
    boneToModify {boneName=Scale_Hair}
   }
   n19 = Get IsShortBackblingOnAlpha() -> (IsShortBackblingOnAlpha:double)
   n20 = "Transform (Modify) Bone - Bone: dyn_hair_tail_02"(ComponentPose=@n18, Alpha=@n19, Translation="0.003,-4.48,0", Rotation="0,0,0", Scale="1,1,1") {
    rotationSpace=BCS_BoneSpace
    translationMode=BMM_Additive
    translationSpace=BCS_BoneSpace
    alphaBoolBlend {blendOption=Linear}
    boneToModify {boneName=dyn_hair_tail_02}
    translation=<0.003 -4.48>
   }
   n21 = "Transform (Modify) Bone - Bone: dyn_hair_tail_03"(ComponentPose=@n20, Alpha=@n19, Translation="0,0,0", Rotation="0.015,-3.671,-0.025", Scale="1,1,1") {
    rotationMode=BMM_Additive
    rotationSpace=BCS_BoneSpace
    translationMode=BMM_Additive
    translationSpace=BCS_BoneSpace
    alphaBoolBlend {blendOption=Linear}
    boneToModify {boneName=dyn_hair_tail_03}
    rotation=<rot 0.015 -3.671 -0.025>
   }
   n22 = "Transform (Modify) Bone - Bone: dyn_hair_tail_04"(ComponentPose=@n21, Alpha=@n19, Translation="-0.005,-0.652,0.064", Rotation="0.003,-3.001,0.008", Scale="1,1,1") {
    rotationMode=BMM_Additive
    rotationSpace=BCS_BoneSpace
    translationMode=BMM_Additive
    translationSpace=BCS_BoneSpace
    alphaBoolBlend {blendOption=Linear}
    boneToModify {boneName=dyn_hair_tail_04}
    rotation=<rot 0.003 -3.001 0.008>
    translation=<-0.005 -0.652 0.064>
   }
   n23 = Get is Jacket Variant() -> (is Jacket Variant:bool)
   n24 = "Transform (Modify) Bone - Bone: dyn_hair_tail_02"!(ComponentPose=@n22, bAlphaBoolEnabled:bool=@n23, Translation="0,-0.833,0", Rotation="0,0,0", Scale="1,1,1") -> (Pose) {
    alphaInputType=Bool
    rotationSpace=BCS_BoneSpace
    translationMode=BMM_Additive
    translationSpace=BCS_BoneSpace
    alphaBoolBlend {blendOption=Linear}
    boneToModify {boneName=dyn_hair_tail_02}
    translation {y=-0.833}
   }
   n25 = "Transform (Modify) Bone - Bone: dyn_hair_tail_03"!(ComponentPose=@n24, bAlphaBoolEnabled:bool=@n23, Translation="0.023,2.995,0.106", Rotation="-0.001,3.240,-0.005", Scale="1,1,1") -> (Pose) {
    alphaInputType=Bool
    rotationMode=BMM_Additive
    rotationSpace=BCS_BoneSpace
    translationMode=BMM_Additive
    translationSpace=BCS_BoneSpace
    alphaBoolBlend {blendOption=Linear}
    boneToModify {boneName=dyn_hair_tail_03}
    rotation=<rot -0.001 3.2399 -0.005>
    translation=<0.023 2.995 0.106>
   }
   n26 = "Transform (Modify) Bone - Bone: dyn_hair_tail_04"!(ComponentPose=@n25, bAlphaBoolEnabled:bool=@n23, Translation="0.051,-2.078,-1.324", Rotation="0.004,-8.156,-0.000", Scale="1,1,1") -> (Pose) {
    alphaInputType=Bool
    rotationMode=BMM_Additive
    rotationSpace=BCS_BoneSpace
    translationMode=BMM_Additive
    translationSpace=BCS_BoneSpace
    alphaBoolBlend {blendOption=Linear}
    boneToModify {boneName=dyn_hair_tail_04}
    rotation {pitch=0.004, yaw=-8.156}
    translation=<0.051 -2.078 -1.324>
   }
   n27 = Get is Jacket Variant()
   n28 = "Transform (Modify) Bone - Bone: Hair_AltSkin_Corrective_01"!(ComponentPose=@n26, bAlphaBoolEnabled:bool=@n27, Translation="0,0,0", Rotation="0,0,0", Scale="1,1,1") -> (Pose) {
    alphaInputType=Bool
    rotationSpace=BCS_BoneSpace
    translationMode=BMM_Additive
    translationSpace=BCS_BoneSpace
    alphaBoolBlend {blendOption=Linear}
    boneToModify {boneName=Hair_AltSkin_Corrective_01}
   }
   n29 = "Transform (Modify) Bone - Bone: Hair_AltSkin_Corrective_02"!(ComponentPose=@n28, bAlphaBoolEnabled:bool=@n27, Translation="0,-3,0", Rotation="-0.008,0.004,0", Scale="1,1,1") -> (Pose) {
    alphaInputType=Bool
    rotationMode=BMM_Additive
    rotationSpace=BCS_BoneSpace
    translationMode=BMM_Additive
    translationSpace=BCS_BoneSpace
    alphaBoolBlend {blendOption=Linear}
    boneToModify {boneName=Hair_AltSkin_Corrective_02}
    rotation {pitch=-0.008, yaw=0.004}
    translation {y=-3}
   }
   n30 = "Transform (Modify) Bone - Bone: Hair_AltSkin_Corrective_03"!(ComponentPose=@n29, bAlphaBoolEnabled:bool=@n27, Translation="0,-2,0", Rotation="-0.011,-0.001,-1.517", Scale="1,1,1") -> (Pose) {
    alphaInputType=Bool
    rotationMode=BMM_Additive
    rotationSpace=BCS_BoneSpace
    translationMode=BMM_Additive
    translationSpace=BCS_BoneSpace
    alphaBoolBlend {blendOption=Linear}
    boneToModify {boneName=Hair_AltSkin_Corrective_03}
    rotation=<rot -0.011 -0.001 -1.517>
    translation {y=-2}
   }
   n31 = "Transform (Modify) Bone - Bone: dyn_hair_tail_04"(
    ComponentPose=@n30
    bAlphaBoolEnabled=Get Is full skydive() -> (Is full skydive:bool)
    Translation="0,6.002,0"
    Rotation="0,-3.785,0"
    Scale="1,1,1"
   ) {
    alphaInputType=Bool
    rotationMode=BMM_Additive
    rotationSpace=BCS_BoneSpace
    translationMode=BMM_Additive
    translationSpace=BCS_BoneSpace
    alphaBoolBlend {blendInTime=0.5, blendOption=Linear, blendOutTime=0.5}
    boneToModify {boneName=dyn_hair_tail_04}
    rotation {yaw=-3.785}
    translation {y=6.002}
   }
   Save cached pose 'Pre sim'(Pose=Component To Local(ComponentPose=@n31)) {cacheName=Pre sim}
  ]}
 EventGraph {nodes [
   Event BlueprintUpdateAnimation()
   > n1 = Sequence!(execute) -> (then_0, then_1, then_2:Exec, then_3:Exec)
   Set Is LOD 2(execute=@n1.then_1, Is LOD 2:bool="integer >= integer"(A=Get CurrentLOD(), B=2)) -> (Output_Get:bool)
   > n2 = Gravity Override State Function DEPRECATED(CharacterPartAnimInstance=Self-Reference(), GameTimeInSeconds=GetGameTimeInSeconds(), GravityOverrideParams=Get Gravity Override Params())
   > Set Gravity Override(Gravity Override=@n2)
   Set Is full skydive(
    execute=@n1.then_2
    Is full skydive:bool=OR Boolean!(A=Get bIsSkyDiving(), B=Get bIsSkydiveDiveMode(), C:bool=Get bIsSkydiveFloating()) -> (ReturnValue)
   ) -> (Output_Get:bool)
   n3 = Get bIsParachuteOpened()
   n4 = Get Is full skydive()
   n5 = Get bIsJumping()
   n6 = Get bIsFalling()
   > Set Is jumping fall(Is jumping fall:bool=AND Boolean!(A=OR Boolean!(A=@n5, B=@n6) -> (ReturnValue), B=NOT Boolean(A=@n3), C:bool=NOT Boolean(A=@n4)) -> (ReturnValue)) -> (Output_Get:bool)
   > n7 = Sequence!(execute) -> (then_0, then_1)
   > Branch()
   > n8 = PlayerMovement(AnimInstance=Self-Reference(), XDotProductCoefficient=1, YDotProductCoefficient=1, ZDotProductCoefficient=1)
   n9 = MapRangeClamped(Value=@n8.YMovement, InRangeA=0.75, InRangeB=1)
   n10 = Get is Wear a Cape()
   > n11 = Branch(Condition=AND Boolean!(A="float >= float"(A=@n9, B=0.1), B=@n10) -> (ReturnValue))
   > Set Gravity Override Params(Gravity Override Params:Gravity Override Params Struct Structure=Get Gravity Override Cape moving() -> (Gravity Override Cape moving:Gravity Override Params Struct Structure)) -> (Output_Get:Gravity Override Params Struct Structure)
   n12 = Branch(execute=@n11.else, Condition=@n10)
   n13 = MapRangeClamped(Value=@n8, InRangeA=-0.75, InRangeB=-1)
   n14 = MapRangeClamped(Value=@n8, InRangeA=0.75, InRangeB=1)
   > Set Moving sides(Moving sides:bool=OR Boolean(A="float >= float"(A=@n13, B=0.1), B="float >= float"(A=@n14, B=0.1))) -> (Output_Get:bool)
   > n15 = Set Moving Backward(Moving Backward:bool="float >= float"(A=MapRangeClamped(Value=@n8.YMovement, InRangeA=-0.75, InRangeB=-1), B=0.1)) -> (Output_Get:bool)
   Set Gravity Override Params(
    execute=[@n15.then, @n12.else]
    Gravity Override Params="Get Gravity Override  default"() -> ("Gravity Override  default":Gravity Override Params Struct Structure)
   )
   Branch(execute=@n7.then_1, Condition=Get bIsInFrontEnd())
   > Set FEIdleFloat(FEIdleFloat:double=GetCurveValue(self=GetOwnerFrontEndAnimInstance(), CurveName=Frontend_Idle_Bypass)) -> (Output_Get:double)
   n16 = Get RootComponent!(self:Actor Object Reference=Get Player Pawn() -> (Player Pawn:Player Pawn Athena Object Reference)) -> (RootComponent:Scene Component Object Reference)
   n17 = Unrotate Vector!(
    A:Vector=GetComponentVelocity!(self:Scene Component Object Reference=@n16) -> (ReturnValue:Vector)
    B:Rotator=Get World Rotation!(self:Scene Component Object Reference=@n16) -> (ReturnValue:Rotator) {is=title K2Node_CallFunction}
   ) -> (ReturnValue_X:double, ReturnValue_Y:double, ReturnValue_Z:double)
   Set Is Moving Forward(execute=@n1.then_0, Is Moving Forward:double=SelectFloat(A=1, B=0, bPickA="float >= float"(A=@n17, B=100))) -> (Output_Get:double)
   n18 = Event BlueprintInitializeAnimation()
   n19 = Sequence(execute=[@n18.then, @n29.then])
   > Set Body_Mesh(Body_Mesh=GetOwnerSkeletalMeshComponent())
   n20 = TryGetPawnOwner()
   > Is Valid(InputObject=@n20)
   > n21 = Cast To FortPlayerPawnAthena(Object=@n20)
   > Set As Fort Player Pawn Athena(As Fort Player Pawn Athena:Fort Player Pawn Athena Object Reference=@n21) -> (Output_Get:Fort Player Pawn Athena Object Reference)
   > n22 = Cast To PlayerPawn_Athena(Object=@n20)
   > Set Player Pawn(Player Pawn:Player Pawn Athena Object Reference=@n22) -> (Output_Get:Player Pawn Athena Object Reference)
   n23 = Sequence(execute=@n19.then_1)
   n24 = GetSkeletalMeshForPartType(self=Get As Fort Player Pawn Athena() -> (As Fort Player Pawn Athena:Fort Player Pawn Athena Object Reference))
   > Set is Jacket Variant(is Jacket Variant:bool="Equal (Object)"(
     A=GetSkeletalMeshAsset!(self:Skeletal Mesh Component Object Reference=@n24) -> (ReturnValue:Skeletal Mesh Object Reference)
     B=/BRCosmetics/Characters/Player/Female/Medium/Bodies/F_MED_BrainMatter_Jacket/Meshes/F_MED_BrainMatter_Jacket
    )) -> (Output_Get:bool)
   n25 = CheckForBackbling()
   n26 = Set IsWearingBackbling(execute=@n23.then_1, IsWearingBackbling:bool=@n25) -> (Output_Get:bool)
   > Set Is Backbling A Cape(Is Backbling A Cape:bool=@n25.bIsBackblingACape) -> (Output_Get:bool)
   > Set NotBackbling(NotBackbling:double="To Float (Boolean)"(InBool=NOT Boolean(A=@n25))) -> (Output_Get:double)
   > Set is Wear a Cape(is Wear a Cape:bool=AND Boolean(A=@n25, B=@n25.bIsBackblingLong)) -> (Output_Get:bool)
   > n27 = Branch(Condition=@n26)
   > n28 = Branch(Condition=@n25.bIsBackblingLong)
   Set IsShortBackblingOnAlpha(execute=[@n28.then, @n27.else], IsShortBackblingOnAlpha:double=0) -> (Output_Get:double)
   Set IsShortBackblingOnAlpha(execute=@n28.else, IsShortBackblingOnAlpha=1)
   n29 = Event OnCharacterCustomizationLoaded()
  ]}
]
variables [
 As Fort Player Pawn Athena {type=FortPlayerPawnAthena}
 Body_Mesh {type=SkeletalMeshComponent}
 FEIdleFloat {type=number}
 Gravity Override {type=Vector}
 "Gravity Override  default" {
  type=GravityOverrideParamsStruct
  default {
   onGround_Moving {gravityOverride=<0 0 -750>}
   skydive_FloatingDown {gravityOverride=<0 -300 550>, windAmplitude=<200 -100 10>, windFrequency=<1.5 7 10>, windOffset=<0 0 0>}
   skydive_GoingUp {gravityOverride=<0 -300 700>, windAmplitude=<200 -100 10>, windFrequency=<1.5 7 10>}
   skydive_Parachuting {gravityOverride=<0 -250 -750>, windAmplitude=<200 -200 100>, windOffset=<0 -200 0>}
  }
 }
 Gravity Override Cape moving {
  type=GravityOverrideParamsStruct
  default {
   onGround_Moving {gravityOverride=<0 -550 -500>, windOffset=<0 400 300>}
   skydive_DivingDown {gravityOverride=<-1500 0 -450>}
   skydive_FloatingDown {gravityOverride=<0 -300 350>, windAmplitude=<200 -100 10>, windFrequency=<1.5 7 10>, windOffset=<0 0 0>}
   skydive_Parachuting {gravityOverride=<0 -250 -750>, windAmplitude=<200 -200 100>, windOffset=<0 -200 0>}
   tactical_Sprint {gravityOverride=<0 -450 -200>}
  }
 }
 Gravity Override Params {
  type=GravityOverrideParamsStruct
  default {onGround_Moving {gravityOverride=<0 0 -750>}, skydive_FloatingDown {gravityOverride=<0 -300 100>}}
 }
 Is Backbling A Cape {type=boolean}
 Is full skydive {type=boolean}
 is Jacket Variant {type=boolean}
 Is jumping fall {type=boolean}
 Is LOD 2 {type=boolean}
 Is Moving Forward {type=number}
 is Wear a Cape {type=boolean}
 IsShortBackblingOnAlpha {type=number}
 IsWearingBackbling {type=boolean}
 Moving Backward {type=boolean}
 Moving sides {type=boolean}
 NotBackbling {type=number}
 Player Pawn {type=PlayerPawn_Athena_C}
]
